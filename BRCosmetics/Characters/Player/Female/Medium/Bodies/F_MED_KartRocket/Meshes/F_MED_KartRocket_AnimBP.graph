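# Post-processes the F_MED_KartRocket female body cosmetic, adding secondary motion to jacket, collar and pouch zippers.
# Parent mesh pose feeds a crouch blend, aiming the pouch at the right calf or constraining it to hip bones, then a bullets constraint.
# Cached PreSim runs through Control Rig into chained AnimDynamics, with stacked zipper passes using differing gravity.
# A duplicate jacket/collar chain is unconnected; the event graph is empty.
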AnimBlueprint /BRCosmetics/Characters/Player/Female/Medium/Bodies/F_MED_KartRocket/Meshes/F_MED_KartRocket_AnimBP
target_skeleton=&/F_MED_KartRocket_Skeleton
graphs [
 AnimGraph {nodes [
   n1 = Anim Dynamics - Bone: dyn_jacket_fr_l(Alpha=1, ExternalForce="0,0,0") {
    angularDampingOverride=1
    angularSpringConstant=100
    bAngularSpring=true
    bOverrideAngularDamping=true
    bUseGravityOverride=true
    simulationSpace=RootRelative
    alphaBoolBlend {blendOption=Linear}
    boundBone {boneName=dyn_jacket_fr_l}
    componentAppliedLinearAccClamp=<5000 5000 2000>
    componentLinearAccScale=<0.4 0.4 0.4>
    componentLinearVelScale=<0.5 0.5 0.5>
    gravityOverride {z=-980}
    physicsBodyDefinitions [{
      boundBone {boneName=dyn_jacket_fr_l}
      boxExtents=<11 11 11>
      constraintSetup {angularLimitsMax=<5 5 5>, angularLimitsMin=<-5 -5 -30>, angularTarget=<1 0 0>}
      localJointOffset=<4.5 -1 0>
     }]
    relativeSpaceBone {boneName=root}
   }
   n2 = Anim Dynamics - Bone: dyn_jacket_l(ComponentPose=@n1, Alpha=1, ExternalForce="0,0,0") {
    angularDampingOverride=0.999
    angularSpringConstant=100
    bAngularSpring=true
    bOverrideAngularDamping=true
    bUseGravityOverride=true
    simulationSpace=RootRelative
    alphaBoolBlend {blendOption=Linear}
    boundBone {boneName=dyn_jacket_l}
    componentAppliedLinearAccClamp=<2000 2000 2000>
    componentLinearAccScale=<0.6 0.6 0.4>
    componentLinearVelScale=<0.7 0.7 0.5>
    gravityOverride {z=-980}
    physicsBodyDefinitions [{
      boundBone {boneName=dyn_jacket_l}
      boxExtents=<10 10 10>
      constraintSetup {angularLimitsMax=<5 10 5>, angularLimitsMin=<-5 -10 -50>, angularTarget=<1 0 0>}
      localJointOffset=<5 -1 0>
     }]
    relativeSpaceBone {boneName=root}
   }
   n3 = Anim Dynamics - Bone: dyn_jacket_bk_l(ComponentPose=@n2, Alpha=1, ExternalForce="0,0,0") {
    angularDampingOverride=1
    angularSpringConstant=1000
    bAngularSpring=true
    bOverrideAngularDamping=true
    bUseGravityOverride=true
    simulationSpace=RootRelative
    alphaBoolBlend {blendOption=Linear}
    boundBone {boneName=dyn_jacket_bk_l}
    componentAppliedLinearAccClamp=<2000 2000 2000>
    componentLinearAccScale=<0.4 0.4 0.4>
    componentLinearVelScale=<0.4 0.4 0.4>
    gravityOverride {z=-980}
    physicsBodyDefinitions [{
      boundBone {boneName=dyn_jacket_bk_l}
      boxExtents=<10 20 20>
      constraintSetup {angularLimitsMax=<5 10 5>, angularLimitsMin=<-5 -10 -50>, angularTarget=<1 0 0>}
      localJointOffset=<3 0 0>
     }]
    relativeSpaceBone {boneName=root}
   }
   n4 = Anim Dynamics - Bone: dyn_jacket_fr_r(ComponentPose=@n3, Alpha=1, ExternalForce="0,0,0") {
    angularDampingOverride=1
    angularSpringConstant=100
    bAngularSpring=true
    bOverrideAngularDamping=true
    bUseGravityOverride=true
    simulationSpace=RootRelative
    alphaBoolBlend {blendOption=Linear}
    boundBone {boneName=dyn_jacket_fr_r}
    componentAppliedLinearAccClamp=<5000 5000 2000>
    componentLinearAccScale=<1 1 0.4>
    componentLinearVelScale=<1 1 0.5>
    gravityOverride {z=-980}
    physicsBodyDefinitions [{
      boundBone {boneName=dyn_jacket_fr_r}
      boxExtents=<12 12 12>
      constraintSetup {angularLimitsMax=<5 5 5>, angularLimitsMin=<-5 -5 -40>, angularTarget=<1 0 0>}
      localJointOffset=<5 -1 0>
     }]
    relativeSpaceBone {boneName=root}
   }
   n5 = Anim Dynamics - Bone: dyn_jacket_r(ComponentPose=@n4, Alpha=1, ExternalForce="0,0,0") {
    angularDampingOverride=0.999
    angularSpringConstant=100
    bAngularSpring=true
    bOverrideAngularDamping=true
    bUseGravityOverride=true
    simulationSpace=RootRelative
    alphaBoolBlend {blendOption=Linear}
    boundBone {boneName=dyn_jacket_r}
    componentAppliedLinearAccClamp=<2000 2000 2000>
    componentLinearAccScale=<0.6 0.6 0.4>
    componentLinearVelScale=<0.7 0.7 0.5>
    gravityOverride {z=-980}
    physicsBodyDefinitions [{
      boundBone {boneName=dyn_jacket_r}
      boxExtents=<10 10 10>
      constraintSetup {angularLimitsMax=<5 10 5>, angularLimitsMin=<-5 -10 -50>, angularTarget=<1 0 0>}
      localJointOffset=<5 -1 0>
     }]
    relativeSpaceBone {boneName=root}
   }
   n6 = Anim Dynamics - Bone: dyn_jacket_bk_r(ComponentPose=@n5, Alpha=1, ExternalForce="0,0,0") {
    angularDampingOverride=1
    angularSpringConstant=1000
    bAngularSpring=true
    bOverrideAngularDamping=true
    bUseGravityOverride=true
    simulationSpace=RootRelative
    alphaBoolBlend {blendOption=Linear}
    boundBone {boneName=dyn_jacket_bk_r}
    componentAppliedLinearAccClamp=<2000 2000 2000>
    componentLinearAccScale=<0.4 0.4 0.4>
    componentLinearVelScale=<0.4 0.4 0.4>
    gravityOverride {z=-980}
    physicsBodyDefinitions [{
      boundBone {boneName=dyn_jacket_bk_r}
      boxExtents=<10 20 20>
      constraintSetup {angularLimitsMax=<5 10 5>, angularLimitsMin=<-5 -10 -50>, angularTarget=<1 0 0>}
      localJointOffset=<3 0 0>
     }]
    relativeSpaceBone {boneName=root}
   }
   n7 = Anim Dynamics - Bone: dyn_collar_l(ComponentPose=@n6, Alpha=1, ExternalForce="0,0,0") {
    angularDampingOverride=1
    angularSpringConstant=100
    bOverrideAngularDamping=true
    bUseGravityOverride=true
    simulationSpace=RootRelative
    alphaBoolBlend {blendOption=Linear}
    boundBone {boneName=dyn_collar_l}
    componentAppliedLinearAccClamp=<250 250 1000>
    componentLinearAccScale=<0.4 0.4 0.4>
    componentLinearVelScale=<0.5 0.5 0.5>
    gravityOverride {z=-980}
    physicsBodyDefinitions [{
      boundBone {boneName=dyn_collar_l}
      boxExtents=<10 10 10>
      constraintSetup {angularLimitsMax=<5 30 20>, angularLimitsMin=<-5 -5 -20>, angularTarget=<1 0 0>}
      localJointOffset=<3 1 0>
     }]
    relativeSpaceBone {boneName=root}
   }
   Anim Dynamics - Bone: dyn_collar_r(ComponentPose=@n7, Alpha=1, ExternalForce="0,0,0") {
    angularDampingOverride=1
    angularSpringConstant=100
    bOverrideAngularDamping=true
    bUseGravityOverride=true
    numSolverIterationsPreUpdate=3
    simulationSpace=RootRelative
    alphaBoolBlend {blendOption=Linear}
    boundBone {boneName=dyn_collar_r}
    componentAppliedLinearAccClamp=<500 500 1000>
    componentLinearAccScale=<0.6 0.6 0.4>
    componentLinearVelScale=<0.7 0.7 0.5>
    gravityOverride {z=-980}
    physicsBodyDefinitions [{
      boundBone {boneName=dyn_collar_r}
      boxExtents=<15 12 12>
      constraintSetup {angularLimitsMax=<5 5 30>, angularLimitsMin=<-5 -30 -30>, angularTarget=<1 0 0>}
      localJointOffset=<5 0.5 0>
     }]
    relativeSpaceBone {boneName=root}
   }
   n8 = Control Rig(Source=Use cached pose 'PreSim'()) {
    alphaBoolBlend {blendOption=Linear}
    controlRigAssetReference {blueprintRigClass=&/F_MED_KartRocket_CtrlRig.F_MED_KartRocket_CtrlRig_C}
   }
   n9 = Anim Dynamics - Bone: dyn_jacket_fr_l(ComponentPose=Local To Component(LocalPose=@n8), Alpha=1, ExternalForce="0,0,0") {
    angularDampingOverride=0.9998
    angularSpringConstant=100
    bAngularSpring=true
    bOverrideAngularDamping=true
    bUseGravityOverride=true
    simulationSpace=RootRelative
    alphaBoolBlend {blendOption=Linear}
    boundBone {boneName=dyn_jacket_fr_l}
    componentAppliedLinearAccClamp=<5000 5000 2000>
    componentLinearAccScale=<0.4 0.4 0.4>
    componentLinearVelScale=<0.5 0.5 0.5>
    gravityOverride {z=-980}
    physicsBodyDefinitions [{
      boundBone {boneName=dyn_jacket_fr_l}
      boxExtents=<11 11 11>
      constraintSetup {angularLimitsMax=<5 5 5>, angularLimitsMin=<-5 -5 -30>, angularTarget=<1 0 0>}
      localJointOffset=<4.5 -1 0>
     }]
    relativeSpaceBone {boneName=root}
   }
   n10 = Anim Dynamics - Bone: dyn_jacket_l(ComponentPose=@n9, Alpha=1, ExternalForce="0,0,0") {
    angularDampingOverride=0.998
    angularSpringConstant=100
    bAngularSpring=true
    bOverrideAngularDamping=true
    bUseGravityOverride=true
    simulationSpace=RootRelative
    alphaBoolBlend {blendOption=Linear}
    boundBone {boneName=dyn_jacket_l}
    componentAppliedLinearAccClamp=<2000 2000 2000>
    componentLinearAccScale=<0.6 0.6 0.4>
    componentLinearVelScale=<0.7 0.7 0.5>
    gravityOverride {z=-980}
    physicsBodyDefinitions [{
      boundBone {boneName=dyn_jacket_l}
      boxExtents=<10 10 10>
      constraintSetup {angularLimitsMax=<5 10 5>, angularLimitsMin=<-5 -10 -50>, angularTarget=<1 0 0>}
      localJointOffset=<5 -1 0>
     }]
    relativeSpaceBone {boneName=root}
   }
   n11 = Anim Dynamics - Bone: dyn_jacket_bk_l(ComponentPose=@n10, Alpha=1, ExternalForce="0,0,0") {
    angularDampingOverride=0.9998
    angularSpringConstant=1000
    bAngularSpring=true
    bOverrideAngularDamping=true
    bUseGravityOverride=true
    simulationSpace=RootRelative
    alphaBoolBlend {blendOption=Linear}
    boundBone {boneName=dyn_jacket_bk_l}
    componentAppliedLinearAccClamp=<2000 2000 2000>
    componentLinearAccScale=<0.4 0.4 0.4>
    componentLinearVelScale=<0.4 0.4 0.4>
    gravityOverride {z=-980}
    physicsBodyDefinitions [{
      boundBone {boneName=dyn_jacket_bk_l}
      boxExtents=<10 20 20>
      constraintSetup {angularLimitsMax=<5 10 5>, angularLimitsMin=<-5 -10 -50>, angularTarget=<1 0 0>}
      localJointOffset=<3 0 0>
     }]
    relativeSpaceBone {boneName=root}
   }
   n12 = Anim Dynamics - Bone: dyn_jacket_fr_r(ComponentPose=@n11, Alpha=1, ExternalForce="0,0,0") {
    angularDampingOverride=0.9997
    angularSpringConstant=100
    bAngularSpring=true
    bOverrideAngularDamping=true
    bUseGravityOverride=true
    simulationSpace=RootRelative
    alphaBoolBlend {blendOption=Linear}
    boundBone {boneName=dyn_jacket_fr_r}
    componentAppliedLinearAccClamp=<5000 5000 2000>
    componentLinearAccScale=<1 1 0.4>
    componentLinearVelScale=<1 1 0.5>
    gravityOverride {z=-980}
    physicsBodyDefinitions [{
      boundBone {boneName=dyn_jacket_fr_r}
      boxExtents=<12 12 12>
      constraintSetup {angularLimitsMax=<5 5 5>, angularLimitsMin=<-5 -5 -40>, angularTarget=<1 0 0>}
      localJointOffset=<5 -1 0>
     }]
    relativeSpaceBone {boneName=root}
   }
   n13 = Anim Dynamics - Bone: dyn_jacket_r(ComponentPose=@n12, Alpha=1, ExternalForce="0,0,0") {
    angularDampingOverride=0.998
    angularSpringConstant=100
    bAngularSpring=true
    bOverrideAngularDamping=true
    bUseGravityOverride=true
    simulationSpace=RootRelative
    alphaBoolBlend {blendOption=Linear}
    boundBone {boneName=dyn_jacket_r}
    componentAppliedLinearAccClamp=<2000 2000 2000>
    componentLinearAccScale=<0.6 0.6 0.4>
    componentLinearVelScale=<0.7 0.7 0.5>
    gravityOverride {z=-980}
    physicsBodyDefinitions [{
      boundBone {boneName=dyn_jacket_r}
      boxExtents=<10 10 10>
      constraintSetup {angularLimitsMax=<5 10 5>, angularLimitsMin=<-5 -10 -50>, angularTarget=<1 0 0>}
      localJointOffset=<5 -1 0>
     }]
    relativeSpaceBone {boneName=root}
   }
   n14 = Anim Dynamics - Bone: dyn_jacket_bk_r(ComponentPose=@n13, Alpha=1, ExternalForce="0,0,0") {
    angularDampingOverride=0.999
    angularSpringConstant=1000
    bAngularSpring=true
    bOverrideAngularDamping=true
    bUseGravityOverride=true
    simulationSpace=RootRelative
    alphaBoolBlend {blendOption=Linear}
    boundBone {boneName=dyn_jacket_bk_r}
    componentAppliedLinearAccClamp=<2000 2000 2000>
    componentLinearAccScale=<0.4 0.4 0.4>
    componentLinearVelScale=<0.4 0.4 0.4>
    gravityOverride {z=-980}
    physicsBodyDefinitions [{
      boundBone {boneName=dyn_jacket_bk_r}
      boxExtents=<10 20 20>
      constraintSetup {angularLimitsMax=<5 10 5>, angularLimitsMin=<-5 -10 -50>, angularTarget=<1 0 0>}
      localJointOffset=<3 0 0>
     }]
    relativeSpaceBone {boneName=root}
   }
   n15 = Anim Dynamics - Bone: dyn_collar_l(ComponentPose=@n14, Alpha=1, ExternalForce="0,0,0") {
    angularDampingOverride=0.999
    angularSpringConstant=100
    bOverrideAngularDamping=true
    bUseGravityOverride=true
    simulationSpace=RootRelative
    alphaBoolBlend {blendOption=Linear}
    boundBone {boneName=dyn_collar_l}
    componentAppliedLinearAccClamp=<250 250 1000>
    componentLinearAccScale=<0.4 0.4 0.4>
    componentLinearVelScale=<0.5 0.5 0.5>
    gravityOverride {z=-980}
    physicsBodyDefinitions [{
      boundBone {boneName=dyn_collar_l}
      boxExtents=<10 10 10>
      constraintSetup {angularLimitsMax=<5 30 20>, angularLimitsMin=<-5 -5 -20>, angularTarget=<1 0 0>}
      localJointOffset=<3 1 0>
     }]
    relativeSpaceBone {boneName=root}
   }
   n16 = Anim Dynamics - Bone: dyn_collar_r(ComponentPose=@n15, Alpha=1, ExternalForce="0,0,0") {
    angularDampingOverride=0.9997
    angularSpringConstant=100
    bOverrideAngularDamping=true
    bUseGravityOverride=true
    numSolverIterationsPreUpdate=3
    simulationSpace=RootRelative
    alphaBoolBlend {blendOption=Linear}
    boundBone {boneName=dyn_collar_r}
    componentAppliedLinearAccClamp=<500 500 1000>
    componentLinearAccScale=<0.6 0.6 0.4>
    componentLinearVelScale=<0.7 0.7 0.5>
    gravityOverride {z=-980}
    physicsBodyDefinitions [{
      boundBone {boneName=dyn_collar_r}
      boxExtents=<15 12 12>
      constraintSetup {angularLimitsMax=<5 5 30>, angularLimitsMin=<-5 -30 -30>, angularTarget=<1 0 0>}
      localJointOffset=<5 0.5 0>
     }]
    relativeSpaceBone {boneName=root}
   }
   n17 = Anim Dynamics - Bone: dyn_pouch_zipper_01(ComponentPose=@n16, Alpha=1, ExternalForce="0,0,0") {
    angularDampingOverride=0.999
    bOverrideAngularDamping=true
    bUseGravityOverride=true
    simulationSpace=RootRelative
    alphaBoolBlend {blendOption=Linear}
    boundBone {boneName=dyn_pouch_zipper_01}
    componentAppliedLinearAccClamp=<200 200 500>
    componentLinearAccScale=<0.2 0.2 0.5>
    componentLinearVelScale=<0.5 0.5 0.8>
    gravityOverride {z=-980}
    physicsBodyDefinitions [{
      boundBone {boneName=dyn_pouch_zipper_01}
      boxExtents=<10 15 10>
      constraintSetup {angularLimitsMax=<1 0 25>, angularLimitsMin=<-1 -40 0>, angularTarget=<0 0 0>}
      localJointOffset=<1 0 0>
     }]
   }
   n18 = Anim Dynamics - Bone: dyn_pouch_zipper_02(ComponentPose=@n17, Alpha=1, ExternalForce="0,0,0") {
    angularDampingOverride=0.999
    bOverrideAngularDamping=true
    bUseGravityOverride=true
    simulationSpace=RootRelative
    alphaBoolBlend {blendOption=Linear}
    boundBone {boneName=dyn_pouch_zipper_02}
    componentAppliedLinearAccClamp=<200 200 500>
    componentLinearAccScale=<0.2 0.2 0.5>
    componentLinearVelScale=<0.5 0.5 0.8>
    gravityOverride {z=-980}
    physicsBodyDefinitions [{
      boundBone {boneName=dyn_pouch_zipper_02}
      boxExtents=<10 10 10>
      constraintSetup {angularLimitsMax=<1 0 25>, angularLimitsMin=<-1 -40 -5>, angularTarget=<0 0 0>}
      localJointOffset=<1 0 0>
     }]
   }
   n19 = Anim Dynamics - Bone: dyn_pouch_zipper_01!(ComponentPose=@n18, bAlphaBoolEnabled:bool=True, ExternalForce="0,0,0") -> (Pose) {
    alphaInputType=Bool
    angularDampingOverride=0.999
    bOverrideAngularDamping=true
    bUseGravityOverride=true
    simulationSpace=RootRelative
    alphaBoolBlend {blendInTime=0.2, blendOption=Linear, blendOutTime=0.2}
    boundBone {boneName=dyn_pouch_zipper_01}
    componentAppliedLinearAccClamp=<200 200 500>
    componentLinearAccScale=<0.2 0.2 0.5>
    componentLinearVelScale=<0.5 0.5 0.8>
    gravityOverride {z=700}
    physicsBodyDefinitions [{
      boundBone {boneName=dyn_pouch_zipper_01}
      boxExtents=<10 15 10>
      constraintSetup {angularLimitsMax=<1 0 50>, angularLimitsMin=<-1 -50 -10>, angularTarget=<0 0 0>}
      localJointOffset=<1 0 0>
     }]
   }
   n20 = Anim Dynamics - Bone: dyn_pouch_zipper_02!(ComponentPose=@n19, bAlphaBoolEnabled:bool=True, ExternalForce="0,0,0") -> (Pose) {
    alphaInputType=Bool
    angularDampingOverride=0.999
    bOverrideAngularDamping=true
    bUseGravityOverride=true
    simulationSpace=RootRelative
    alphaBoolBlend {blendInTime=0.2, blendOption=Linear, blendOutTime=0.2}
    boundBone {boneName=dyn_pouch_zipper_02}
    componentAppliedLinearAccClamp=<200 200 500>
    componentLinearAccScale=<0.2 0.2 0.5>
    componentLinearVelScale=<0.5 0.5 0.8>
    gravityOverride {z=700}
    physicsBodyDefinitions [{
      boundBone {boneName=dyn_pouch_zipper_02}
      boxExtents=<10 10 10>
      constraintSetup {angularLimitsMax=<1 0 50>, angularLimitsMin=<-1 -50 -10>, angularTarget=<0 0 0>}
      localJointOffset=<1 0 0>
     }]
   }
   n21 = Anim Dynamics - Bone: dyn_pouch_zipper_01(ComponentPose=@n20, bAlphaBoolEnabled=True, ExternalForce="0,0,0") {
    alphaInputType=Bool
    angularDampingOverride=0.999
    bOverrideAngularDamping=true
    bUseGravityOverride=true
    simulationSpace=RootRelative
    alphaBoolBlend {blendInTime=0.2, blendOption=Linear, blendOutTime=0.2}
    boundBone {boneName=dyn_pouch_zipper_01}
    componentAppliedLinearAccClamp=<200 200 500>
    componentLinearAccScale=<0.2 0.2 0.5>
    componentLinearVelScale=<0.5 0.5 0.8>
    gravityOverride {z=1500}
    physicsBodyDefinitions [{
      boundBone {boneName=dyn_pouch_zipper_01}
      boxExtents=<10 15 10>
      constraintSetup {angularLimitsMax=<1 30 50>, angularLimitsMin=<-1 -50 -10>, angularTarget=<0 0 0>}
      localJointOffset=<1 0 0>
     }]
   }
   n22 = Anim Dynamics - Bone: dyn_pouch_zipper_02(ComponentPose=@n21, bAlphaBoolEnabled=True, ExternalForce="0,0,0") {
    alphaInputType=Bool
    angularDampingOverride=0.999
    bOverrideAngularDamping=true
    bUseGravityOverride=true
    simulationSpace=RootRelative
    alphaBoolBlend {blendInTime=0.2, blendOption=Linear, blendOutTime=0.2}
    boundBone {boneName=dyn_pouch_zipper_02}
    componentAppliedLinearAccClamp=<200 200 500>
    componentLinearAccScale=<0.2 0.2 0.5>
    componentLinearVelScale=<0.5 0.5 0.8>
    gravityOverride {z=1500}
    physicsBodyDefinitions [{
      boundBone {boneName=dyn_pouch_zipper_02}
      boxExtents=<10 10 10>
      constraintSetup {angularLimitsMax=<1 30 50>, angularLimitsMin=<-1 -50 -10>, angularTarget=<0 0 0>}
      localJointOffset=<1 0 0>
     }]
   }
   Output Pose(Result=Component To Local(ComponentPose=@n22)) {layerGroup=None}
   n23 = Look At - Bone: deform_pouch(ComponentPose=Local To Component(LocalPose=Copy Pose From Mesh() {bUseAttachedParent=true}), Alpha=1) {
    alphaBoolBlend {blendOption=Linear}
    boneToModify {boneName=deform_pouch}
    lookAtLocation {x=0, z=10}
    lookAtTarget {boneReference {boneName=calf_r}}
    lookAt_Axis {axis=<1 0>}
   }
   n24 = Component To Local(ComponentPose=@n23)
   n25 = "Constraint - deform_pouch_MSTR (PPPPP)"!(
    ComponentPose=Local To Component(LocalPose=Copy Pose From Mesh() {bUseAttachedParent=true})
    Alpha=1
    ConstraintWeights_0=0.013
    ConstraintWeights_1:float=0.034
    ConstraintWeights_2:float=0.124
    ConstraintWeights_3:float=0.184
    ConstraintWeights_4:float=0.645
   ) -> (Pose) {
    alphaBoolBlend {blendOption=Linear}
    boneToModify {boneName=deform_pouch_MSTR}
    constraintSetup [
     {targetBone {boneName=deform_glute_r}}
     {targetBone {boneName=deform_groin_r}}
     {targetBone {boneName=pelvis}}
     {targetBone {boneName=thigh_r}}
     {targetBone {boneName=thigh_twist_01_r}}
    ]
    constraintWeights [1,1,1,1,1]
   }
   n26 = Component To Local(ComponentPose=@n25)
   n27 = Get bIsCrouching()
   n28 = "Constraint - deform_bullets_MSTR (PPPPP)"!(
    ComponentPose=Local To Component(LocalPose=Blend Poses by bool(BlendPose_0=@n24, BlendPose_1=@n26, bActiveValue=@n27) {blendType=Linear})
    Alpha=1
    ConstraintWeights_0=0.164
    ConstraintWeights_1:float=0.625
    ConstraintWeights_2:float=0.031
    ConstraintWeights_3:float=0.013
    ConstraintWeights_4:float=0.167
   ) -> (Pose) {
    alphaBoolBlend {blendOption=Linear}
    boneToModify {boneName=deform_bullets_MSTR}
    constraintSetup [
     {targetBone {boneName=deform_groin_r}}
     {targetBone {boneName=pelvis}}
     {targetBone {boneName=spine_02}}
     {targetBone {boneName=thigh_twist_01_l}}
     {targetBone {boneName=thigh_twist_01_r}}
    ]
    constraintWeights [1,1,1,1,1]
   }
   Save cached pose 'PreSim'(Pose=Component To Local(ComponentPose=@n28)) {cacheName=PreSim}
  ]}
 EventGraph {nodes [Event BlueprintUpdateAnimation()]}
]
variables []
