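# Physics asset for the F_MED_KartRocket female face accessory mesh, defining bodies for physics on its left and right dynamic hair bones.
# Each side starts at a small master sphere: the left carries one capsule, the right a chain of two, linked by swing and twist limited constraints.
# Unconstrained neck and upper-spine capsules provide body shapes alongside the hair chains.
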PhysicsAsset /BRCosmetics/Characters/Player/Female/Medium/Bodies/F_MED_KartRocket/Meshes/Parts/F_MED_KartRocket_FaceAcc_Physics
bodies {
 dyn_hair_l_01 [{shapeType=Capsule, length=4.5, radius=1.5, shapeName=dyn_hair_l_01_capsule, center=<2.0543 0 -1.7011>, rotation=<rot 0.0003 179.9987 -180>}]
 dyn_hair_l_mstr [{shapeType=Sphere, radius=1, shapeName=dyn_hair_l_mstr_sphere}]
 dyn_hair_r_01 [{shapeType=Capsule, length=3, radius=2, shapeName=dyn_hair_r_01_capsule, center=<2.0808 0 0>}]
 dyn_hair_r_02 [{shapeType=Capsule, length=3.8554, radius=2.4861, shapeName=dyn_hair_r_02_capsule, center=<4 0 0>, rotation=<rot 90 0 0>}]
 dyn_hair_r_mstr [{shapeType=Sphere, radius=1, shapeName=dyn_hair_r_mstr_sphere}]
 neck_02 [{shapeType=Capsule, length=6.4975, radius=4.2316, shapeName=neck_02_capsule, rotation=<rot -90 -0.0006 0>}]
 spine_05 [{shapeType=Capsule, length=15.5, radius=11, shapeName=spine_05_capsule, center=<1.1721 -1.9559 0>, rotation=<rot 0 -90 0>}]
}
constraints [
 {bone1Name=dyn_hair_r_01, bone2Name=dyn_hair_r_mstr, swing1LimitDegrees=45, swing2LimitDegrees=30, twistLimitDegrees=2}
 {bone1Name=dyn_hair_l_01, bone2Name=dyn_hair_l_mstr, swing1LimitDegrees=30, swing2LimitDegrees=15, twistLimitDegrees=2}
 {bone1Name=dyn_hair_r_02, bone2Name=dyn_hair_r_01, swing1LimitDegrees=20, swing2LimitDegrees=10, twistLimitDegrees=1}
]
